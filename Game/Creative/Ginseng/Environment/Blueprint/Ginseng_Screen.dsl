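; Ginseng quiz screen Blueprint that displays round number, question, portrait and three color-coded answers.
; Custom events load a row from the Ginseng_QuestionAnswers data table, then fill text renders, placing the correct answer by segment color.
; Further events toggle visibility to show the round, question and answers, reveal only the correct answer, and clear the round.
; The construction script applies answer materials, and OnRep_NewRound? also hides question and answer text.
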(event EventBeginPlay
  (Rendering|SetVisibility (Variables|Default|GetSM_Answer1) false false))

(event Custom|EventNewRound
  (params (Round# :type EGinseng_Rounds%20Enum :default "NewEnumerator0"))
  (bind _result (Math|Integer|IncrementInt (Math|Conversions|ToInteger(Byte) Round#))
    (:%20%20
      (Rendering|Components|TextRender|SetText (Variables|Default|GetRound#_c) (Utilities|Text|ToText(Integer) _result))
      (Rendering|SetVisibility (Variables|Default|GetRound_c) true false))))

(event Custom|EventHideRoundTile
  (Rendering|SetVisibility (Variables|Default|GetRound_c) false false))

(event Custom|EventSetQ&AText
  (params (RowName :type Name))
  (bind _returnvalue (Utilities|GetDataTableRow "/Game/Creative/Ginseng/Ginseng_QuestionAnswers.Ginseng_QuestionAnswers" RowName))
  (Variables|Default|SetQuestion (Utilities|Struct|BreakGinsengQuestionsandAnswers _returnvalue))
  (Variables|Default|SetCorrect%20Answer (.CorrectAnswer_17_86BD60264A3B5D78594CE6AB11F23D95 (Utilities|Struct|BreakGinsengQuestionsandAnswers _returnvalue)))
  (Variables|Default|SetWrong%20Answer%201 (.WrongAnswer1_18_EBF44A58429769D53DB59E9A750E1CB2 (Utilities|Struct|BreakGinsengQuestionsandAnswers _returnvalue)))
  (Variables|Default|SetWrong%20Answer%202 (.WrongAnswer2_21_028E9C714A0C446D7C8074A7A74EA0A7 (Utilities|Struct|BreakGinsengQuestionsandAnswers _returnvalue))))

(event Custom|EventSetCorrectAnswer
  (params (CorrectAnswerColor :type EGinseng_SegmentColor%20Enum :default "NewEnumerator0"))
  (Variables|Default|SetCorrect%20Answer%20Color CorrectAnswerColor)
  (Rendering|Components|TextRender|SetText (Variables|Default|GetQuestion_c) (Variables|Default|GetQuestion))
  (CallFunction|SetAnswerstoTextRender))

(event Custom|EventShowQuestion
  (Rendering|SetVisibility (Variables|Default|GetQuestion_c) true false)
  (Rendering|SetVisibility (Variables|Default|GetPortrait) true false))

(event Custom|EventShowAllAnswers
  (Rendering|SetVisibility (Variables|Default|GetAnswer1_c) true false)
  (Rendering|SetVisibility (Variables|Default|GetAnswer2_c) true false)
  (Rendering|SetVisibility (Variables|Default|GetAnswer3_c) true false))

(event Custom|EventRevealCorrectAnswer
  (switch Utilities|FlowControl|Switch|SwitchonEGinseng_SegmentColor (Variables|Default|GetCorrect%20Answer%20Color)
    (:NewEnumerator0
      (Rendering|SetVisibility (Variables|Default|GetAnswer1_c) true false)
      (Rendering|SetVisibility (Variables|Default|GetAnswer2_c) false false)
      (Rendering|SetVisibility (Variables|Default|GetAnswer3_c) false false))
    (:NewEnumerator1
      (Rendering|SetVisibility (Variables|Default|GetAnswer2_c) true false)
      (Rendering|SetVisibility (Variables|Default|GetAnswer1_c) false false)
      (Rendering|SetVisibility (Variables|Default|GetAnswer3_c) false false))
    (:NewEnumerator2
      (Rendering|SetVisibility (Variables|Default|GetAnswer3_c) true false)
      (Rendering|SetVisibility (Variables|Default|GetAnswer1_c) false false)
      (Rendering|SetVisibility (Variables|Default|GetAnswer2_c) false false))))

(event Custom|EventEndRound
  (Rendering|SetVisibility (Variables|Default|GetQuestion_c) false false)
  (Rendering|SetVisibility (Variables|Default|GetAnswer1_c) false false)
  (Rendering|SetVisibility (Variables|Default|GetAnswer2_c) false false)
  (Rendering|SetVisibility (Variables|Default|GetAnswer3_c) false false)
  (Rendering|SetVisibility (Variables|Default|GetPortrait) false false))

(fn OnRep_NewRound?
  (Rendering|SetVisibility (Variables|Default|GetQuestion_c) false false)
  (Rendering|SetVisibility (Variables|Default|GetAnswer1_c) false false)
  (Rendering|SetVisibility (Variables|Default|GetAnswer2_c) false false)
  (Rendering|SetVisibility (Variables|Default|GetAnswer3_c) false false))

(fn SetAnswersToTextRender
  (bind _correct_answer (Variables|Default|GetCorrect%20Answer))
  (bind _wrong_answer_1 (Variables|Default|GetWrong%20Answer%201))
  (bind _wrong_answer_2 (Variables|Default|GetWrong%20Answer%202))
  (switch Utilities|FlowControl|Switch|SwitchonEGinseng_SegmentColor (Variables|Default|GetCorrect%20Answer%20Color)
    (:NewEnumerator0
      (Rendering|Components|TextRender|SetText (Variables|Default|GetAnswer1_c) _correct_answer)
      (Rendering|Components|TextRender|SetText (Variables|Default|GetAnswer2_c) _wrong_answer_1)
      (Rendering|Components|TextRender|SetText (Variables|Default|GetAnswer3_c) _wrong_answer_2))
    (:NewEnumerator1
      (Rendering|Components|TextRender|SetText (Variables|Default|GetAnswer2_c) _correct_answer)
      (Rendering|Components|TextRender|SetText (Variables|Default|GetAnswer1_c) _wrong_answer_1)
      (Rendering|Components|TextRender|SetText (Variables|Default|GetAnswer3_c) _wrong_answer_2))
    (:NewEnumerator2
      (Rendering|Components|TextRender|SetText (Variables|Default|GetAnswer3_c) _correct_answer)
      (Rendering|Components|TextRender|SetText (Variables|Default|GetAnswer1_c) _wrong_answer_1)
      (Rendering|Components|TextRender|SetText (Variables|Default|GetAnswer2_c) _wrong_answer_2))))

(fn ConstructionScript
  (bind (_coloroptions_blue_3_4f16ed9942ab22004d976086024fe33a _coloroptions_green_5_86bd60264a3b5d78594ce6ab11f23d95 _coloroptions_red_7_ebf44a58429769d53db59e9a750e1cb2) (Utilities|Struct|BreakGinsengColorOptions (Variables|GetColorOptions_Blue_3_4F16ED9942AB22004D976086024FE33A)))
  (Rendering|Material|SetMaterial (Variables|Default|GetSM_Answer1) (Math|Conversions|ToInteger(Byte) (Variables|Default|GetAnswer%20Color)) _coloroptions_blue_3_4f16ed9942ab22004d976086024fe33a)
  (Rendering|Material|SetMaterial (Variables|Default|GetSM_Answer2) (Math|Conversions|ToInteger(Byte) (Variables|Default|GetAnswer%20Color)) _coloroptions_green_5_86bd60264a3b5d78594ce6ab11f23d95)
  (Rendering|Material|SetMaterial (Variables|Default|GetSM_Answer3) (Math|Conversions|ToInteger(Byte) (Variables|Default|GetAnswer%20Color)) _coloroptions_red_7_ebf44a58429769d53db59e9a750e1cb2))
