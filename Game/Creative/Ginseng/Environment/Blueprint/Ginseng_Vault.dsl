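; Ginseng vault Blueprint that runs a looping trivia game with colored voting pads and answer screens.
; A one-time sphere overlap gathers pads, screens and data table questions, then starts a timer-chained event sequence that steps GameState through each round.
; OnRep_GameState switches on that state to update pads and screens and cycle the round number.
; Authority randomly picks the correct color and question, removing used questions from the pool.
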(event OnComponentBeginOverlap(Sphere)
  (params (OverlappedComponent :type Primitive%20Component%20Object%20Reference) (OtherActor :type Actor%20Object%20Reference) (OtherComp :type Primitive%20Component%20Object%20Reference) (OtherBodyIndex :type Integer) (bFromSweep :type Boolean) (SweepResult :type Hit%20Result%20Structure%20%28by%20ref%29))
  (Utilities|Casting|CastToFortPlayerPawn OtherActor)
  ;[0.6291 1 0.6795] Start up logic to set variables.
  (Utilities|FlowControl|DoOnce false
    (:Completed
      (bind _outactors (Actor|GetAllActorsOfClass "/Game/Creative/Ginseng/Environment/Blueprint/Ginseng_VotingPad.Ginseng_VotingPad_C"))
      (Variables|Default|SetVoting_Pads _outactors)
      (bind _outactors_2 (Actor|GetAllActorsOfClass "/Game/Creative/Ginseng/Environment/Blueprint/Ginseng_Screen.Ginseng_Screen_C"))
      (Variables|Default|SetScreens _outactors_2)
      (Variables|Default|SetGameState "NewEnumerator0")
      (Variables|Default|SetNewRound# "NewEnumerator0")
      (bind _outrownames (DataTable|GetDataTableRowNames (Variables|Default|GetQuestions&AnswerList)))
      (Variables|Default|SetQuestions&AnswerArray _outrownames)
      (Utilities|Time|SetTimerbyEvent (EventDispatchers|CreateEvent :!event Custom|Event_HideRoundTitle) 1.0))))

(event Custom|Event_HideRoundTitle
  (Variables|Default|SetGameState "NewEnumerator10")
  (Utilities|Time|SetTimerbyEvent (EventDispatchers|CreateEvent :!event Custom|Event_ChooseAnswerColor) 2.0))

(event Custom|Event_ChooseAnswerColor
  (Networking|SwitchHasAuthority
    (:Authority
      (Variables|Default|SetCorrectAnswerColor_RNG (Utilities|Enum|BytetoEnumEGinseng_SegmentColor (Math|Conversions|ToByte(Integer) (Math|Random|RandomInteger (Utilities|Array|Length (Variables|Default|GetVoting_Pads))))))
      (Utilities|Time|SetTimerbyEvent (EventDispatchers|CreateEvent :!event Custom|Event_SetSegmentColor) 1.0))))

(event Custom|Event_SetSegmentColor
  (Variables|Default|SetGameState "NewEnumerator13")
  (CallFunction|EventChooseQuestion (Variables|Default|GetCorrectAnswerColor_RNG)))

(event Custom|Event_ChooseQuestion
  (params (CorrectAnswerColor_RNG :type EGinseng_SegmentColor%20Enum :default "NewEnumerator0"))
  (Variables|Default|SetCorrectAnswerColor CorrectAnswerColor_RNG)
  (Variables|Default|SetGameState "NewEnumerator6")
  (CallFunction|EventSendQuestionInfo (Variables|Default|GetChoosenQuestion_Authority) (Variables|Default|GetQuestion#_Authority)))

(event Custom|Event_SendQuestionInfo
  (params (ChoosenQuestion :type Name) (Question# :type Integer))
  (Variables|Default|SetChoosenQuestion ChoosenQuestion)
  (Variables|Default|SetQuestion# Question#)
  (Variables|Default|SetGameState "NewEnumerator9")
  (CallFunction|EventSetAnswer))

(event Custom|Event_SetAnswer
  ;[1 0.5197 0.4297] Set Correct Answer   PAD Logic
  (Variables|Default|SetGameState "NewEnumerator1")
  (CallFunction|EventStartQuestion))

(event Custom|Event_StartQuestion
  ;[0.2634 0.4476 1] Start Question   SCREEN Logic
  (Variables|Default|SetGameState "NewEnumerator2")
  (Utilities|Time|SetTimerbyEvent (EventDispatchers|CreateEvent :!event Custom|Event_ShowMultipeChoice) 5.0))

(event Custom|Event_ShowMultipeChoice
  ;[0.2634 0.4476 1] Show Multiple Choice   SCREEN Logic
  (Variables|Default|SetGameState "NewEnumerator3")
  (Utilities|Time|SetTimerbyEvent (EventDispatchers|CreateEvent :!event Custom|Event_RevealCorrectAnswer) 1e+01))

(event Custom|Event_RevealCorrectAnswer
  ;[0.2634 0.4476 1] Reveal Correct Answer   SCREEN Logic
  (Variables|Default|SetGameState "NewEnumerator4")
  (CallFunction|EventDisableWrongAnswer))

(event Custom|Event_DisableWrongAnswer
  ;[1 0.5197 0.4297] Disable Wrong Answer  PAD Logic
  (Variables|Default|SetGameState "NewEnumerator5")
  (Utilities|Time|SetTimerbyEvent (EventDispatchers|CreateEvent :!event Custom|Event_EndRound) 3.0))

(event Custom|Event_EndRound
  ;[1 0.5197 0.4297] Set New Round  PAD
  (Variables|Default|SetGameState "NewEnumerator11")
  (Utilities|Time|SetTimerbyEvent (EventDispatchers|CreateEvent :!event Custom|Event_SetNewRound) 5.0))

(event Custom|Event_SetNewRound
  ;[1 0.5197 0.4297] Set New Round  PAD
  (Variables|Default|SetGameState "NewEnumerator0")
  (Utilities|Time|SetTimerbyEvent (EventDispatchers|CreateEvent :!event Custom|Event_HideRoundTitle) 1.5))

(fn OnRep_Correct_VotingPad)

(fn OnRep_CorrectAnswerColor)

(fn OnRep_CorrectAnswerColor_RNG)

(fn OnRep_GameState
  (bind _questions_answerarray (Variables|Default|GetQuestions&AnswerArray))
  (bind (_outitem _outindex) (Utilities|Array|RandomArrayItem _questions_answerarray))
  (bind _newround (Variables|Default|GetNewRound#))
  (switch Utilities|FlowControl|Switch|SwitchonEGinseng_GameState (Variables|Default|GetGameState)
    (:NewEnumerator0
      (for _array_element (Variables|Default|GetVoting_Pads)
        (Class|GinsengVotingPad|EventSetSegmentColor _array_element true "NewEnumerator5"))
      ;[1 1 1] Set which round it is.
      (if (not (Variables|Default|GetisFirstRound?))
        (switch Utilities|FlowControl|Switch|SwitchonEGinseng_Rounds (Variables|Default|GetCurrentRound#)
          (:NewEnumerator0
            (Variables|Default|SetNewRound# "NewEnumerator1"))
          (:NewEnumerator1
            (Variables|Default|SetNewRound# "NewEnumerator2"))
          (:NewEnumerator2
            (Variables|Default|SetNewRound# "NewEnumerator3"))
          (:NewEnumerator3
            (Variables|Default|SetNewRound# "NewEnumerator0"))))
      ;[1 1 1] Set New Round visibility on board segments and screens.
      (Utilities|Array|Shuffle (Variables|Default|GetVoting_Pads))
      (for _array_element_2 (Variables|Default|GetVoting_Pads)
        (Class|GinsengVotingPad|EventSetRound _array_element_2 _newround))
      (for _array_element_3 (Variables|Default|GetScreens)
        (Class|GinsengScreen|EventNewRound _array_element_3 _newround))
      (Variables|Default|SetisFirstRound? false)
      (return))
    (:NewEnumerator10
      (for _array_element_4 (Variables|Default|GetScreens)
        (Class|GinsengScreen|EventHideRoundTile _array_element_4)))
    (:NewEnumerator7
      ;[1 1 1] Select which voting segment will hold. Get the total amount of voting segments and randomly pickying which color will hold the correct answer.
      (Networking|SwitchHasAuthority
        (:Authority
          (Variables|Default|SetCorrectAnswerColor_RNG (Utilities|Enum|BytetoEnumEGinseng_SegmentColor (Math|Conversions|ToByte(Integer) (Math|Random|RandomInteger (Utilities|Array|Length (Variables|Default|GetVoting_Pads)))))))))
    (:NewEnumerator6
      ;[1 1 1] Select the Question from the array. Once Chosen remove it from the array list of possible options.
      (Networking|SwitchHasAuthority
        (:Authority
          (Variables|Default|SetChoosenQuestion_Authority _outitem)
          (Variables|Default|SetQuestion#_Authority _outindex))))
    (:NewEnumerator13
      (for (_array_element_5 _array_index) (Variables|Default|GetVoting_Pads)
        (Class|GinsengVotingPad|EventSetSegmentColor _array_element_5 false (Utilities|Enum|BytetoEnumEGinseng_SegmentColor (Math|Conversions|ToByte(Integer) _array_index)))))
    (:NewEnumerator9
      (for _array_element_6 (Variables|Default|GetScreens)
        (Class|GinsengScreen|EventSetQ&AText _array_element_6 (Variables|Default|GetChoosenQuestion)))
      (Networking|SwitchHasAuthority
        (:Authority
          (Utilities|Array|RemoveIndex (Variables|Default|GetQuestions&AnswerArray) (Variables|Default|GetQuestion#)))))
    (:NewEnumerator1
      ;[1 1 1] Match Segment color with the Correct Answer color on screen. All other segements will be marked as being the Wrong Answer.
      (for _array_element_7 (Variables|Default|GetVoting_Pads)
        (Class|GinsengVotingPad|EventIsAnswerColor _array_element_7 (Variables|Default|GetCorrectAnswerColor)))
      (for _array_element_8 (Variables|Default|GetScreens)
        (Class|GinsengScreen|EventSetCorrectAnswer _array_element_8 (Variables|Default|GetCorrectAnswerColor)))
      (return))
    (:NewEnumerator2
      ;[1 1 1] Reveal the question to the players.
      (for _array_element_9 (Variables|Default|GetScreens)
        (Class|GinsengScreen|EventShowQuestion _array_element_9))
      (Variables|Default|SetCurrentRound# (Variables|Default|GetNewRound#))
      (return))
    (:NewEnumerator3
      ;[1 1 1] Reveal all possible answers
      (for _array_element_10 (Variables|Default|GetScreens)
        (Class|GinsengScreen|EventShowAllAnswers _array_element_10))
      (return))
    (:NewEnumerator4
      ;[1 1 1] Reveal the Correct Answer
      (for _array_element_11 (Variables|Default|GetScreens)
        (Class|GinsengScreen|EventRevealCorrectAnswer _array_element_11))
      (return))
    (:NewEnumerator5
      ;[1 1 1] Disable Wrong Answer Segments
      (for _array_element_12 (Variables|Default|GetVoting_Pads)
        (Class|GinsengVotingPad|EventDisableWrongAnswer _array_element_12))
      (Variables|Default|SetCorrectAnswerColor_RNG "NewEnumerator5")
      (return))
    (:NewEnumerator11
      (for _array_element_13 (Variables|Default|GetScreens)
        (Class|GinsengScreen|EventEndRound _array_element_13)))))

(fn OnRep_New
  (for _array_element (Variables|Default|GetVoting_Pads)
    (Class|GinsengVotingPad|EventDisableWrongAnswer _array_element))
  (return))

(fn OnRep_Round#)

(fn OnRep_Value)

(fn OnRep_Voting_Pads)

(fn OnRep_WrongAnswer
  (for _array_element (Variables|Default|GetVoting_Pads)
    (Class|GinsengVotingPad|EventDisableWrongAnswer _array_element))
  (return))

(fn ConstructionScript)
