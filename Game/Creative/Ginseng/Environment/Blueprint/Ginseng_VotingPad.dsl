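; Ginseng voting pad Blueprint: a ringed, color-coded answer pad for a round-based voting game.
; Construction and a color event apply the segment color's material to Ring_3.
; Setting the round reveals that round's ring with collision through DoOnce gates, hiding Ring_1 in later rounds.
; An answer check compares colors to set a wrong-answer flag, which a separate event uses to disable Ring_1; both OnRep functions are empty.
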(event Custom|Event_SetSegmentColor
  (params (bSetBlank :type Boolean) (Selection :type EGinseng_SegmentColor%20Enum :default "NewEnumerator0"))
  (bind (_segement_color_blue_3_4f16ed9942ab22004d976086024fe33a _segement_color_green_5_86bd60264a3b5d78594ce6ab11f23d95 _segement_color_red_7_ebf44a58429769d53db59e9a750e1cb2 _segement_color_orange_9_028e9c714a0c446d7c8074a7a74ea0a7 _segement_color_purple_11_fa40cca849f03b1a1e6f63b3c6551d51 _segement_color_grey_17_639ee79143c04013524d6e8c62b62965) (Utilities|Struct|BreakGinsengColorOptions (Variables|GetSegement_Color_Blue_3_4F16ED9942AB22004D976086024FE33A)))
  (if (not bSetBlank)
    (Variables|Default|SetSegmentColor Selection))
  (switch Utilities|FlowControl|Switch|SwitchonEGinseng_SegmentColor Selection
    (:NewEnumerator0
      (Rendering|Material|SetMaterial (Variables|Default|GetRing_3) 1 _segement_color_blue_3_4f16ed9942ab22004d976086024fe33a))
    (:NewEnumerator1
      (Rendering|Material|SetMaterial (Variables|Default|GetRing_3) 1 _segement_color_green_5_86bd60264a3b5d78594ce6ab11f23d95))
    (:NewEnumerator2
      (Rendering|Material|SetMaterial (Variables|Default|GetRing_3) 1 _segement_color_red_7_ebf44a58429769d53db59e9a750e1cb2))
    (:NewEnumerator3
      (Rendering|Material|SetMaterial (Variables|Default|GetRing_3) 1 _segement_color_orange_9_028e9c714a0c446d7c8074a7a74ea0a7))
    (:NewEnumerator4
      (Rendering|Material|SetMaterial (Variables|Default|GetRing_3) 1 _segement_color_purple_11_fa40cca849f03b1a1e6f63b3c6551d51))
    (:NewEnumerator5
      (Rendering|Material|SetMaterial (Variables|Default|GetRing_3) 1 _segement_color_grey_17_639ee79143c04013524d6e8c62b62965))))

(event Custom|EventIsAnswerColor
  (params (CorrectAnswerColor :type EGinseng_SegmentColor%20Enum :default "NewEnumerator0"))
  (if (Math|Byte|Equal(Byte) CorrectAnswerColor (Variables|Default|GetSegmentColor))
    (Variables|Default|SetWrongAnswer? false)
    (else
      (Variables|Default|SetWrongAnswer? true))))

(event Custom|Event%20SetRound
  (params (Round# :type EGinseng_Rounds%20Enum :default "NewEnumerator0"))
  ;[0.3927 1 0.3889] Set board segments visibility based on Round
  (Variables|Default|SetRound# Round#)
  (CallFunction|BoardSegmentVisibility (Variables|Default|GetRound#)))

(event Custom|Event%20DisableWrongAnswer
  (bind _ring_1 (Variables|Default|GetRing_1))
  ;[1 0.1303 0.22] Disable Wrong Answer Rings
  (if (Variables|Default|GetWrongAnswer?)
    (Rendering|SetVisibility _ring_1 false false)
    (Collision|SetCollisionEnabled _ring_1 "NoCollision")))

(fn BoardSegmentVisibility
  (params (Selection :type EGinseng_Rounds%20Enum :default "NewEnumerator0"))
  (bind _ring_1 (Variables|Default|GetRing_1))
  (bind _ring_1_2 (Variables|Default|GetRing_1))
  (bind _ring_1_3 (Variables|Default|GetRing_1))
  (bind _ring_3 (Variables|Default|GetRing_3))
  (bind _ring_1_4 (Variables|Default|GetRing_1))
  (bind _ring_2 (Variables|Default|GetRing_2))
  (bind _ring_4 (Variables|Default|GetRing_4))
  (switch Utilities|FlowControl|Switch|SwitchonEGinseng_Rounds Selection
    (:NewEnumerator0
      ;[0.3787 1 0.428] Segment's visibility to turn on based on round.
      (Utilities|FlowControl|DoOnce false
        (:Completed
          (Rendering|SetVisibility _ring_1 true false)
          (Collision|SetCollisionEnabled _ring_1 "QueryAndPhysics"))))
    (:NewEnumerator1
      (Utilities|FlowControl|DoOnce false
        (:Completed
          (Rendering|SetVisibility _ring_2 true false)
          (Collision|SetCollisionEnabled _ring_2 "QueryAndPhysics")
          ;[1 0.07009 0] Segment's visibility to turn off based on round.
          (Rendering|SetVisibility _ring_1_2 false false)
          (Collision|SetCollisionEnabled _ring_1_2 "NoCollision"))))
    (:NewEnumerator2
      (Utilities|FlowControl|DoOnce false
        (:Completed
          (Rendering|SetVisibility _ring_3 true false)
          (Collision|SetCollisionEnabled _ring_3 "QueryAndPhysics")
          (Rendering|SetVisibility _ring_1_4 false false)
          (Collision|SetCollisionEnabled _ring_1_4 "NoCollision"))))
    (:NewEnumerator3
      (Utilities|FlowControl|DoOnce false
        (:Completed
          (Rendering|SetVisibility _ring_4 true false)
          (Collision|SetCollisionEnabled _ring_4 "QueryAndPhysics")
          (Rendering|SetVisibility _ring_1_3 false false)
          (Collision|SetCollisionEnabled _ring_1_3 "NoCollision"))))))

(fn OnRep_Round#)

(fn OnRep_WrongAnswer?)

(fn ConstructionScript
  (bind (_segement_color_blue_3_4f16ed9942ab22004d976086024fe33a _segement_color_green_5_86bd60264a3b5d78594ce6ab11f23d95 _segement_color_red_7_ebf44a58429769d53db59e9a750e1cb2 _segement_color_orange_9_028e9c714a0c446d7c8074a7a74ea0a7 _segement_color_purple_11_fa40cca849f03b1a1e6f63b3c6551d51) (Utilities|Struct|BreakGinsengColorOptions (Variables|GetSegement_Color_Blue_3_4F16ED9942AB22004D976086024FE33A)))
  (switch Utilities|FlowControl|Switch|SwitchonEGinseng_SegmentColor (Variables|Default|GetSegmentColor)
    (:NewEnumerator0
      (Rendering|Material|SetMaterial (Variables|Default|GetRing_3) 1 _segement_color_blue_3_4f16ed9942ab22004d976086024fe33a))
    (:NewEnumerator1
      (Rendering|Material|SetMaterial (Variables|Default|GetRing_3) 1 _segement_color_green_5_86bd60264a3b5d78594ce6ab11f23d95))
    (:NewEnumerator2
      (Rendering|Material|SetMaterial (Variables|Default|GetRing_3) 1 _segement_color_red_7_ebf44a58429769d53db59e9a750e1cb2))
    (:NewEnumerator3
      (Rendering|Material|SetMaterial (Variables|Default|GetRing_3) 1 _segement_color_orange_9_028e9c714a0c446d7c8074a7a74ea0a7))
    (:NewEnumerator4
      (Rendering|Material|SetMaterial (Variables|Default|GetRing_3) 1 _segement_color_purple_11_fa40cca849f03b1a1e6f63b3c6551d51))))
