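# Material_002 is a minimal Default Lit surface material in the Ginseng environment content, with no clear purpose beyond what looks like a flat placeholder.
# Its graph holds a single vector parameter with a purple default and shows no connection from it to a material output.
# Usage flags enable it on skeletal meshes and mesh deformers.
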Material /Game/Creative/Ginseng/Environment/Materials/Material_002
graph {
 properties {bUsedWithMeshDeformer=true, bUsedWithSkeletalMesh=true, pixelDepthOffsetMode=PDOM_Legacy, usedShadingModels=Default Lit}
 nodes [VectorParameter() {defaultValue=<rgba 0.4442 0 0.8229 0>}]
}
