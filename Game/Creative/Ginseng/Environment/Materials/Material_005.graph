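# Material_005 is a simple Default Lit surface material among the Ginseng environment assets.
# Its graph holds a single vector parameter with a reddish-orange default, with no textures, math or shown connection to a material output.
# Usage flags enable it on skeletal meshes and with mesh deformers.
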Material /Game/Creative/Ginseng/Environment/Materials/Material_005
graph {
 properties {bUsedWithMeshDeformer=true, bUsedWithSkeletalMesh=true, pixelDepthOffsetMode=PDOM_Legacy, usedShadingModels=Default Lit}
 nodes [VectorParameter() {defaultValue=<rgba 0.8958 0.1656 0 0>}]
}
