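# Minimal Default Lit material in the Ginseng environment materials folder, with a generic name that does not indicate which surface it covers.
# Its graph holds a single vector parameter left at default settings, with no textures, math or visible link to a material output.
# Usage flags enable it on skeletal meshes and with mesh deformers, and pixel depth offset uses legacy mode.
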Material /Game/Creative/Ginseng/Environment/Materials/Material_008
graph {
 properties {bUsedWithMeshDeformer=true, bUsedWithSkeletalMesh=true, pixelDepthOffsetMode=PDOM_Legacy, usedShadingModels=Default Lit}
 nodes [VectorParameter()]
}
