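# Translucent Default Lit material in the Creative Ginseng environment folder, referencing a Fortnite-style BananaAgent soldier hero icon texture.
# Its graph holds a single TextureSample of that UI icon, with no connection to any material output shown.
# The generic name Material_012 and unwired sampler leave its in-game purpose unclear.
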Material /Game/Creative/Ginseng/Environment/Materials/Material_012
graph {
 properties {blendMode=BLEND_Translucent, pixelDepthOffsetMode=PDOM_Legacy, usedShadingModels=Default Lit}
 nodes [TextureSample() {texture=/Game/UI/Foundation/Textures/Icons/Heroes/Athena/Soldier/T-Soldier-HID-701-Athena-Commando-M-BananaAgent-L}]
}
